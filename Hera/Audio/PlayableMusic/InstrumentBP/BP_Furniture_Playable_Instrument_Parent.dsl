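; Parent Blueprint for furniture instruments where each player interaction plays the next note of a song, with optional disabling on mobile.
; BeginPlay maps the instrument to an integer, which PlayNote passes with the current note to MetaSound parameters and triggers.
; The last step triggers an end flourish, blocks interaction briefly, and advances through five looping songs.
; Interaction also sets DummyBool and forces a net update.
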(event EventTick
  (params (DeltaSeconds :type Float%20%28single-precision%29)))

(event EventBeginPlay
  (bind _instrument (Variables|PlayableInstrument|GetInstrument))
  (bind _instrumentintmap (Variables|PlayableInstrument|GetInstrumentIntMap))
  (bind (_value _returnvalue) (Utilities|Map|Find _instrumentintmap _instrument))
  (if (and (Variables|PlayableInstrument|GetDisableInteractOnMobile) (TouchPlatform|IsMobileGame))
    (Variables|Interaction|SetbAllowInteract false)
    (else
      (Variables|Interaction|SetbAllowInteract true)
      (if _returnvalue
        (Variables|PlayableInstrument|SetInstrumentInt _value)))))

(event Collision|EventActorBeginOverlap
  (params (OtherActor :type Actor%20Object%20Reference)))

(event Custom|PlayNote
  (params (InstrumentInt :type Integer) (CurrentNote :type Integer) (SongNotOver :type Boolean) (MusicComp :type Audio%20Component%20Object%20Reference))
  (bind _playedmusic (Variables|Default|GetPlayedMusic))
  (bind _playedmusic_2 (Variables|Default|GetPlayedMusic))
  (Utilities|IsValid _playedmusic_2
    (:Is%20Valid
      (Variables|/Script/Engine.AudioComponent|GetSound _playedmusic_2
        (:then
          _)
        (:else
          (bind _returnvalue (Audio|SpawnSoundAttached "/Hera/Audio/PlayableMusic/MSS_PlayableInstrument.MSS_PlayableInstrument" (Variables|Transformation|GetRootComponent) "None" (Variables|PlayableInstrument|GetAudioComponentOffset)))
          (Variables|Default|SetPlayedMusic _returnvalue)))
      (Audio|Parameter|SetIntegerParameter _playedmusic "Instrument" InstrumentInt :!class /Script/Engine.AudioComponent)
      (if SongNotOver
        (Audio|Parameter|SetIntegerParameter _playedmusic "CurrentNote" CurrentNote :!class /Script/Engine.AudioComponent)
        (Audio|Parameter|ExecuteTriggerParameter _playedmusic "PlayNote")
        (else
          (Audio|Parameter|ExecuteTriggerParameter _playedmusic "PlayEndFlourish")
          (Variables|Interaction|SetbAllowInteract false)
          (Utilities|FlowControl|Delay 2.0)
          (Variables|Interaction|SetbAllowInteract true))))
    (:Is%20Not%20Valid
      _)))

(event Interaction|EventBlueprintOnInteract
  (params (InteractingPawn :type Fort%20Pawn%20Object%20Reference) (InteractionBeingAttempted :type EInteractionBeingAttempted%20Enum :default "FirstInteraction"))
  (bind _currentsongnotes (Variables|PlayableInstrument|GetCurrentSongNotes))
  (bind _returnvalue (Utilities|Array|Length _currentsongnotes))
  (bind _returnvalue_2 (- _returnvalue 1))
  (bind _noteindex (Variables|PlayableInstrument|GetNoteIndex))
  (bind _returnvalue_3 (< _noteindex _returnvalue_2))
  (if (Variables|PlayableInstrument|GetDummyBool)
    _)
  (Variables|PlayableInstrument|SetDummyBool true)
  (Networking|ForceNetUpdate)
  ;[1 0 0 0.5] NEW SONG??? SEE HERE!!!
  (switch int (Variables|PlayableInstrument|GetCurrentSong)
    (:0
      (Variables|PlayableInstrument|SetCurrentSongNotes (Variables|Songs|GetOGTheme)))
    (:1
      (Variables|PlayableInstrument|SetCurrentSongNotes (Variables|Songs|GetButterBarn)))
    (:2
      (Variables|PlayableInstrument|SetCurrentSongNotes (Variables|Songs|GetSwimFree)))
    (:3
      (Variables|PlayableInstrument|SetCurrentSongNotes (Variables|Songs|GetMakeTheFishstickPay)))
    (:4
      (Variables|PlayableInstrument|SetCurrentSongNotes (Variables|Songs|GetWaddleWalk))))
  (CallFunction|PlayNote (Variables|PlayableInstrument|GetInstrumentInt) (Utilities|Array|Get(aref) _currentsongnotes _noteindex) _returnvalue_3)
  (if _returnvalue_3
    (Variables|PlayableInstrument|SetNoteIndex (+ (Variables|PlayableInstrument|GetNoteIndex) 1))
    (else
      (Variables|PlayableInstrument|SetNoteIndex 0)
      ; Go to next song or cycle back around
      (if (< (Variables|PlayableInstrument|GetCurrentSong) 4)
        (Variables|PlayableInstrument|SetCurrentSong (+ (Variables|PlayableInstrument|GetCurrentSong) 1))
        (else
          (Variables|PlayableInstrument|SetCurrentSong 0))))))

(fn BlueprintCanInteract
  (params (InteractingPawn :type Fort%20Pawn%20Object%20Reference) (InteractionBeingAttempted :type EInteractionBeingAttempted%20Enum :default "FirstInteraction") (InteractionType :type TInteractionType%20Enum :default "IT_NoInteraction"))
  (return true))

(fn BlueprintGetInteractionString
  (params (InteractingPawn :type Fort%20Pawn%20Object%20Reference) (InteractionBeingAttempted :type EInteractionBeingAttempted%20Enum :default "FirstInteraction"))
  (return "NSLOCTEXT(\"[38955C31F49081E29FB69AA6A10CA18E]\", \"532E50184D1F95B111B6168C70D6F3B7\", \"Play\")"))

(fn CanDisplayEasyInteractButton
  (return true))

(fn OnRep_DummyBool)

(fn ConstructionScript)
